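# MF_SharedSnow is a winter terrain material function that builds a snow layer and blends it over input material attributes.
# World-XY projected snow masks at two scales are height-blended with base Opacity, then weighted by vertex-normal Z.
# Layered snow and noise normals plus mask-driven color and roughness form the layer, blended by that mask or, via static switch, height-blended inverted vertex red.
# Intermediate snow channels are also output; several branches, including fresnel color, are unconnected.
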MaterialFunction /Figment_S07/LM_7_00/Environments/Sets/Winter/Terrain/MF_SharedSnow
graph {nodes [
  n1 = ScalarParameter() {defaultValue=-2.5, parameterName=SnowNormalMask03_1}
  n2 = ScalarParameter() {defaultValue=1.5, parameterName=SnowNormalMask03_2}
  n3 = ConstantBiasScale(None=ComponentMask(None=VertexNormalWS()) {b=true})
  n4 = FunctionInput(Preview=MakeMaterialAttributes!(
    BaseColor=Constant3Vector() {constant=<rgba 0 1 0 0>}
    Metallic
    Specular
    Roughness
    Anisotropy
    EmissiveColor
    Opacity=Constant()
    OpacityMask
    Normal=Constant3Vector() {constant=<rgba 0 0 1 0>}
    Tangent
    WorldPositionOffset
    SubsurfaceColor
    ClearCoat
    ClearCoatRoughness
    AmbientOcclusion
    Refraction
    CustomizedUVs_0
    CustomizedUVs_1
    CustomizedUVs_2
    CustomizedUVs_3
    CustomizedUVs_4
    CustomizedUVs_5
    CustomizedUVs_6
    CustomizedUVs_7
    PixelDepthOffset
    ShadingModel
    Displacement
   )) {bUsePreviewValueAsDefault=true, inputName=MaterialLayers, inputType=FunctionInput_MaterialAttributes, sortPriority=1}
  n5 = BreakMaterialAttributes!(Attr=@n4)
  n6 = ComponentMask(None=@n5.Normal) {b=true}
  n7 = MaterialFunctionCall(
   In=Power(Base=ConstantBiasScale(None=@n6), Exp=ScalarParameter() {defaultValue=1.3, parameterName=SnowNormalMask01})
   Contrast=ScalarParameter() {defaultValue=3, parameterName=SnowNormalMask02}
  ) {materialFunction=&CheapContrast}
  n8 = ScalarParameter() {defaultValue=3, parameterName=SnowNormalMask02_1}
  Clamp(None=Add(A=LinearInterpolate(A=@n1, B=@n2, Alpha=@n3), B=Multiply(A=@n7.Result, B=@n8)))
  n9 = WorldPosition() {worldPositionShaderOffset=WPT_ExcludeAllShaderOffsets}
  n10 = ScalarParameter() {defaultValue=-30, group=Cliffs, parameterName=SnowScaleA_5}
  n11 = TextureSample(UVs=ComponentMask(None=Divide(A=@n9.XYZ, B=@n10) {constB=-512}) {g=true, r=true}) {samplerType=SAMPLERTYPE_Normal, texture=/Figment_S07/LM_7_00/Environments/World/Backgrounds/Locales/Temperate/Textures/T_WaterNoise}
  Clamp(None=@n11.B)
  Constant3Vector() {constant=<rgba -0.25 -0.25 0 1>}
  Constant3Vector() {constant=<rgba 1 1 0.5 1>}
  Constant3Vector() {constant=<rgba 0.5 0.5 0.5 1>}
  Constant() {r=-500}
  n12 = VectorParameter() {parameterName=SnowColor, defaultValue=<rgba 0.7262 0.7933 0.83 1>}
  n13 = ScalarParameter() {defaultValue=5, parameterName=SnowDiffuseMaskPower}
  n14 = ComponentMask(None=Divide(
    A=WorldPosition() {worldPositionShaderOffset=WPT_ExcludeAllShaderOffsets}.XYZ
    B=ScalarParameter() {defaultValue=-1800, group=Cliffs, parameterName=SnowScaleA_1}
   ) {constB=-512}) {g=true, r=true}
  n15 = TextureSample(UVs=@n14) {texture=&/T_Snow01_M}
  n16 = ComponentMask(None=Divide(
    A=WorldPosition() {worldPositionShaderOffset=WPT_ExcludeAllShaderOffsets}.XYZ
    B=ScalarParameter() {defaultValue=-400, group=Cliffs, parameterName=SnowScaleA_4}
   ) {constB=-512}) {g=true, r=true}
  n17 = TextureSample(UVs=@n16) {texture=&T_Snow01_M}
  n18 = MaterialFunctionCall(
   Transition Phase=Power(Base=@n15.R, Exp=ScalarParameter() {defaultValue=0.2, group=SnowBlend, parameterName=SnowBlend_Contrast_ContrastTransition})
   Height Texture=@n17.R
  ) {materialFunction=&HeightLerp}
  n19 = MaterialFunctionCall(
   Transition Phase=LinearInterpolate(
    A=ScalarParameter() {defaultValue=0.5, group=SnowBlend, parameterName=SnowBlend_Contrast_Min}
    B=Constant() {r=1}
    Alpha=@n18.Lerp Alpha No Contrast
   )
   Height Texture=Multiply(A=ComponentMask(None=@n5.Opacity) {r=true}, B=ScalarParameter() {defaultValue=0.2, group=SnowBlend, parameterName=SnowBlend_Contrast_HBias})
  ) {materialFunction=&HeightLerp}
  n20 = ScalarParameter() {defaultValue=-0.9849, parameterName=SnowBlendMask_1}
  n21 = ScalarParameter() {defaultValue=1.6904, parameterName=SnowBlendMask_2}
  n22 = ComponentMask(None=VertexNormalWS()) {b=true}
  n23 = Multiply(A=@n19.Lerp Alpha No Contrast, B=Clamp(None=LinearInterpolate(A=@n20, B=@n21, Alpha=@n22)))
  n24 = MaterialFunctionCall(In=@n23, Contrast=ScalarParameter() {defaultValue=0.5, group=Cliffs, parameterName=specpwr1_2}) {materialFunction=&CheapContrast}
  n25 = Multiply(A=@n23, B=@n24.Result)
  n26 = ScalarParameter() {defaultValue=0.5, parameterName=SnowSpec}
  n27 = LinearInterpolate(
   A=ScalarParameter() {defaultValue=0.1, parameterName=SnowRough_2}
   B=ScalarParameter() {defaultValue=0.7, parameterName=SnowRough_1}
   Alpha=Power(Base=ConstantBiasScale(None=@n11.B), Exp=ScalarParameter() {defaultValue=5, group=Cliffs, parameterName=specpwr1_1})
  )
  n28 = Constant3Vector() {constant=<rgba 0 0 1 0>}
  n29 = If(
   A=DotProduct(A=@n28, B=Constant3Vector() {constant=<rgba 0 0 1 0>}) {desc=generate sign for this projection}
   B=Constant()
   "A > B"=Constant() {r=1}
   "A < B"=Constant() {r=-1}
  )
  n30 = Constant() {r=-1}
  n31 = Constant() {r=1}
  n32 = Multiply(
   A=VectorParameter() {parameterName=Normal2Amount, defaultValue=<rgba 0.5 0.5 0.5 1>}.RGB
   B=TextureSample(UVs=@n16) {samplerType=SAMPLERTYPE_Normal, texture=&/T_Snow01_N}.RGB
  )
  n33 = TextureSample(UVs=@n14) {samplerType=SAMPLERTYPE_Normal, texture=&T_Snow01_N}
  n34 = Multiply(A=VectorParameter() {parameterName=Normal3Amount, defaultValue=<rgba 0.5 0.5 0.5 1>}.RGB, B=@n33.RGB)
  n35 = VectorParameter() {parameterName=Normal3Amount_1, defaultValue=<rgba 0.5 0.5 1 1>}
  n36 = Add(
   A=Multiply(A=VectorParameter() {parameterName=Normal1Amount, defaultValue=<rgba -0.25 -0.25 0 1>}.RGB, B=@n11.RGB)
   B=MaterialFunctionCall(A=Add(A=@n32, B=@n34), B=Multiply(A=@n33.RGB, B=@n35.RGB), Transition Phase=@n15.R, Height Texture=@n17.R) {materialFunction=&HeightLerp}.Results
  )
  n37 = MaterialFunctionCall!(Vector1=@n28, Vector2=Constant3Vector() {constant=<rgba 0 1 0 0>}) {materialFunction=/Engine/Functions/Engine_MaterialFunctions02/Math/CreateThirdOrthogonalVector}
  n38 = MaterialFunctionCall(
   VectorToTransform=Multiply(
    A=Normalize(VectorInput=@n36)
    B=MaterialFunctionCall(X=@n29, Y=@n30, Z=@n31) {materialFunction=&MakeFloat3}.Result
   )
   BasisX=@n37.Vector3
   BasisY=@n37.Vector3
   BasisZ=@n37.Vector3
  ) {
   desc=lerps in when the vertex normal faces X or Negative X
   materialFunction=&Transform3x3Matrix
  }
  n39 = MakeMaterialAttributes(
   BaseColor=LinearInterpolate(A=VectorParameter() {parameterName=SnowColor_1, defaultValue=<rgba 0.2539 0.3438 0.375 1>}.RGB, B=@n12.RGB, Alpha=Power(Base=@n25, Exp=@n13))
   Specular=@n26
   Roughness=LinearInterpolate(A=ScalarParameter() {defaultValue=-0.5, parameterName=IceRough}, B=@n27, Alpha=@n25)
   Opacity=@n25
   Normal=LinearInterpolate(A=@n5.Normal, B=@n38.Transformed Vector, Alpha=@n24.Result)
  )
  FunctionOutput(None=@n39) {outputName=OutputSnowMaterial}
  n40 = OneMinus(None=VertexColor().R)
  n41 = ScalarParameter() {defaultValue=-0.081, parameterName=BlendOffset}
  n42 = MaterialFunctionCall(
   Transition Phase=Clamp(None=Add(A=@n40, B=@n41))
   Height Texture=@n17.R
   Contrast=ScalarParameter() {defaultValue=0.99, parameterName=VertexColorBlendContrast}
  ) {materialFunction=&HeightLerp}
  n43 = FunctionInput(Preview=StaticBool()) {bUsePreviewValueAsDefault=true, inputName=UseVertexBlend, inputType=FunctionInput_StaticBool, sortPriority=342}
  FunctionOutput(None=MaterialFunctionCall!(Base Material=@n4, Top Material=@n39, Alpha=StaticSwitch(True=@n42.Alpha, False=@n23, Value=@n43)) {materialFunction=/Engine/Functions/MaterialLayerFunctions/MatLayerBlend_Standard}.Blended Material) {outputName=OutputBlendedMaterial}
  FunctionOutput(None=@n38.Transformed Vector) {outputName=OutputNormals}
  FunctionOutput(None=@n27) {outputName=OutputRoughness}
  FunctionOutput(None=@n26) {outputName=OutputSpec}
  FunctionOutput(None=@n17.R) {outputName=OutputZMask}
  FunctionOutput(None=ComponentMask(None=@n12.RGB) {b=true, g=true, r=true}) {outputName=OutputRegularColor}
  FunctionOutput(None=@n18.Lerp Alpha No Contrast) {outputName=OutputHeightmap}
  LinearInterpolate(
   A=@n12.RGB
   B=VectorParameter() {parameterName=SnowFresnelColor, defaultValue=<rgba 0.496 0.7253 0.8 1>}.RGB
   Alpha=Fresnel(ExponentIn=ScalarParameter() {defaultValue=1, parameterName=SnowFresnelExpo})
  )
  n44 = VectorParameter() {parameterName=asdfasdf3, defaultValue=<rgba 3 3 1 1>}
  n45 = WorldPosition() {worldPositionShaderOffset=WPT_ExcludeAllShaderOffsets}
  n46 = Reroute()
  n47 = TextureSample(UVs=ComponentMask(None=Divide(A=@n45.XYZ, B=@n46) {constB=-512}) {g=true, r=true}) {samplerType=SAMPLERTYPE_Normal, texture=/Figment_S07/LM_7_00/VisualThreatManager/StormVisuals/LargeClouds/T_StormCloudDetail01_N}
  n48 = VertexNormalWS()
  n49 = If(
   A=DotProduct(A=@n48, B=Constant3Vector() {constant=<rgba 0 0 1 0>}) {desc=generate sign for this projection}
   B=Constant()
   "A > B"=Constant() {r=1}
   "A < B"=Constant() {r=-1}
  )
  n50 = Constant() {r=-1}
  n51 = Constant() {r=1}
  n52 = MaterialFunctionCall(Vector1=@n48, Vector2=Constant3Vector() {constant=<rgba 0 1 0 0>}) {materialFunction=&CreateThirdOrthogonalVector}
  MaterialFunctionCall(
   VectorToTransform=Multiply(
    A=Multiply(A=@n44.RGB, B=@n47.RGB)
    B=MaterialFunctionCall(X=@n49, Y=@n50, Z=@n51) {materialFunction=&MakeFloat3}.Result
   )
   BasisX=@n52.Vector3
   BasisY=@n52.Vector3
   BasisZ=@n52.Vector3
  ) {
   desc=lerps in when the vertex normal faces X or Negative X
   materialFunction=&Transform3x3Matrix
  }
  Multiply(A=@n19.Lerp Alpha No Contrast)
  Multiply(A=@n18.Lerp Alpha No Contrast)
  n53 = ScalarParameter() {defaultValue=1, parameterName=CliffDesat_3}
  n54 = ScalarParameter() {defaultValue=1, parameterName=CliffDesat_4}
  n55 = Multiply(A=Clamp(None=LinearInterpolate(A=@n53, B=@n54, Alpha=@n17.R)), B=@n15.R)
  OneMinus(None=Power(Base=OneMinus(None=@n55), Exp=ScalarParameter() {defaultValue=10, parameterName=CliffDesat_2}))
  Power(Base=@n3, Exp=ScalarParameter() {defaultValue=1, parameterName=SnowNormalMask03})
  Transform(None=@n5.Normal)
  Transform(None=@n28)
  VertexNormalWS()
  WorldPosition()
  WorldPosition()
  WorldPosition()
 ]}
